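# Unlit translucent particle material for the Moth pickaxe FX that draws a soft volumetric sphere around each particle position.
# A Custom node ray-intersects the sphere, clips it to scene depth (depth test is disabled), and integrates Inigo Quilez-style analytic density.
# Density drives alpha with particle alpha and camera depth fade, and blends particle color toward a brighter core.
# Dynamic parameters control sphere scale, color falloff, core brightness and density falloff.
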Material /BRCosmetics/Weapons/FORT_Melee/Pickaxe_Moth/FX/M_Moth_SphericalDensity_Sphere
graph {
 properties {
  bDisableDepthTest=true
  bUsedWithInstancedStaticMeshes=true
  bUsedWithMeshParticles=true
  blendMode=BLEND_Translucent
  pixelDepthOffsetMode=PDOM_Legacy
  shadingModel=MSM_Unlit
  usedShadingModels=Unlit
 }
 nodes [
  MaterialExpressionCameraPositionWS_8 = CameraPositionWS() {desc=Ray Origin}
  MaterialExpressionCameraVectorWS_2 = CameraVectorWS() {desc=Ray Direction}
  MaterialExpressionParticlePositionWS_0 = ParticlePositionWS()
  n1 = DynamicParameter() {defaultValue=<rgba 5 1 1 1>, paramNames [ObjectScale, ColorPow, ColorCoreBrghtMult, DensityPower]}
  MaterialExpressionMultiply_58 = Multiply(A=@n1.ObjectScale, B=ScalarParameter() {defaultValue=70, parameterName=ObjectUnitXYRadius})
  MaterialExpressionMaterialFunctionCall_41 = MaterialFunctionCall(Vector 3=Subtract(
    A=CameraPositionWS() {desc=Ray Origin}
    B=MaterialFunctionCall() {materialFunction=&WorldPositionBehindTranslucency}.Result
   )) {materialFunction=/Engine/Functions/Engine_MaterialFunctions02/Utility/VectorLength}
  n2 = Custom!(
   ro=@MaterialExpressionCameraPositionWS_8
   rd=@MaterialExpressionCameraVectorWS_2
   sc=@MaterialExpressionParticlePositionWS_0
   sr=@MaterialExpressionMultiply_58
   depth=@MaterialExpressionMaterialFunctionCall_41.V3 Length
  ) {
   code="// Sphere intersection\r\nfloat3 sd = ro - sc;\r\nfloat b = 2 * dot(rd, sd);\r\nfloat c = dot(sd, sd) - (sr * sr);\r\nfloat h = b * b - 4 * c;\r\n\r\nfloat density = 0.0;\r\n\r\n// Assume hit sphere\r\nh = max(h, 0);\r\n//if (h >= 0)\r\n{\r\n  // Find intersection points\r\n  h = sqrt(h);\r\n  float t1 = (b - h) * 0.5;\r\n  float t2 = (b + h) * 0.5;\r\n\r\n  float bValid = (t2 >= 0.0 &&&& t1 <= depth);\r\n  //if (t2 >= 0.0 &&&& t1 <= depth )\r\n  {\r\n    t1 = max(t1, 0);\r\n    t2 = min(t2, depth);\r\n\r\n    // Linear density\r\n    //density = max(t2 - t1, 0);\r\n    //density /= sr * 2.0;\r\n\r\n    // IQ density\r\n    t1 = -t1 / sr;\r\n    t2 = -t2 / sr;\r\n    sd = (ro - sc) / sr;\r\n    b = dot(rd, sd);\r\n    c = dot(sd, sd) - 1.0;\r\n\r\n    float i1 = (c*t1 + b*t1*t1 + t1*t1*t1/3.0);\r\n    float i2 = (c*t2 + b*t2*t2 + t2*t2*t2/3.0);\r\n    density = (i2-i1)*(3.0/4.0);\r\n  }\r\n\r\n  density *= bValid;\r\n}\r\n\r\nreturn density;"
   description=Sphere Density
   outputType=CMOT_Float1
   inputs [
    {
     inputName=ro
     input {expression=&self:MaterialExpressionCameraPositionWS_8}
    }
    {
     inputName=rd
     input {expression=&self:MaterialExpressionCameraVectorWS_2}
    }
    {
     inputName=sc
     input {expression=&self:MaterialExpressionParticlePositionWS_0}
    }
    {
     inputName=sr
     input {expression=&self:MaterialExpressionMultiply_58}
    }
    {
     inputName=depth
     input {expression=&self:MaterialExpressionMaterialFunctionCall_41}
    }
   ]
  }
  n3 = Multiply(A=Power(Base=@n2, Exp=@n1.DensityPower), B=ParticleColor().A)
  n4 = MaterialFunctionCall(
   Fade Length=ScalarParameter() {defaultValue=32, parameterName=Camera Depth Fade Length}
   Fade Offset=ScalarParameter() {defaultValue=10, parameterName=Camera Depth Fade Offset}
  ) {materialFunction=&CameraDepthFade}
  Clamp(None=Multiply(A=Clamp(None=@n3), B=Clamp(None=@n4.Result)))
  n5 = ParticleColor()
  n6 = Clamp(None=@n2)
  MaterialFunctionCall(Emissive=LinearInterpolate(A=@n5.RGB, B=Multiply(A=@n5.RGB, B=@n1.ColorCoreBrghtMult), Alpha=Power(Base=@n6, Exp=@n1.ColorPow))) {materialFunction=&MF_FX_EmissiveAdjustments}
 ]
}
