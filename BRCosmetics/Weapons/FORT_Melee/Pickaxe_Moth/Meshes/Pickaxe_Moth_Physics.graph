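# Physics asset for the Moth pickaxe cosmetic, a two-handed melee weapon with dangling parts set up for physics.
# The handle body combines several capsules and anchors the shade, a three-bone cord chain and a three-bone cord loop chain.
# The shade and upper cord joints have swing limits with little or no twist allowance, while the cord tip and every loop joint have free swing and twist.
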PhysicsAsset /BRCosmetics/Weapons/FORT_Melee/Pickaxe_Moth/Meshes/Pickaxe_Moth_Physics
bodies {
 Melee_TwoHanded_Handle [
  {shapeType=Capsule, length=81.3029, radius=2.0744, center=<0.2238 -0.3609 5.8577>, rotation=<rot 0.4975 -90 -0.5916>}
  {shapeType=Capsule, length=11, radius=2, center=<3.7297 3.6552 -40.3327>, rotation=<rot 24.386 78.8616 11.4743>}
  {shapeType=Capsule, length=8.375, radius=2.375, center=<3.3062 -5.4677 -39.4181>, rotation=<rot 0.6475 26.5714 28.1406>}
  {shapeType=Capsule, length=8.5, radius=2.125, center=<-4.2209 0 -39.7382>, rotation=<rot -28.0894 0 0>}
 ]
 dyn_Shade [{shapeType=Capsule, length=11.7898, radius=5.9107, center=<-9.5866 0 0>, rotation=<rot 0 90 90>}]
 dyn_cordLoop_01 [{shapeType=Sphere, radius=3.0758, center=<2.5968 0.1917 0.2942>}]
 dyn_cordLoop_02 [{shapeType=Sphere, radius=1.8683, center=<2.1064 -0.0273 -0.1002>}]
 dyn_cordLoop_03 [{shapeType=Sphere, radius=2.0212, center=<1.47 0.0688 -0.0675>}]
 dyn_cord_01 [{shapeType=Capsule, length=1.2352, radius=0.875, center=<1.3189 -0.0371 -0.2318>, rotation=<rot 1.1111 86.11 -83.5095>}]
 dyn_cord_02 [{shapeType=Capsule, length=3.1278, radius=1.125, center=<2.6983 0.036 0.182>, rotation=<rot 0 -90.2586 -89.3837>}]
 dyn_cord_03 [{shapeType=Sphere, radius=2.2657, center=<2.0654 0 0>}]
}
constraints [
 {bone1Name=dyn_Shade, bone2Name=Melee_TwoHanded_Handle, swing1LimitDegrees=20, swing2LimitDegrees=20, twistLimitDegrees=1}
 {bone1Name=dyn_cord_01, bone2Name=Melee_TwoHanded_Handle, swing1LimitDegrees=20, swing2LimitDegrees=20, twistLimitDegrees=0}
 {bone1Name=dyn_cord_02, bone2Name=dyn_cord_01, swing1LimitDegrees=20, swing2LimitDegrees=20, twistLimitDegrees=0}
 {
  bone1Name=dyn_cord_03
  bone2Name=dyn_cord_02
  swing1LimitDegrees=10
  swing1Motion=Free
  swing2LimitDegrees=10
  swing2Motion=Free
  twistLimitDegrees=0
  twistMotion=Free
 }
 {
  bone1Name=dyn_cordLoop_01
  bone2Name=Melee_TwoHanded_Handle
  swing1LimitDegrees=45
  swing1Motion=Free
  swing2LimitDegrees=45
  swing2Motion=Free
  twistLimitDegrees=45
  twistMotion=Free
 }
 {
  bone1Name=dyn_cordLoop_02
  bone2Name=dyn_cordLoop_01
  swing1LimitDegrees=45
  swing1Motion=Free
  swing2LimitDegrees=45
  swing2Motion=Free
  twistLimitDegrees=45
  twistMotion=Free
 }
 {
  bone1Name=dyn_cordLoop_03
  bone2Name=dyn_cordLoop_02
  swing1LimitDegrees=45
  swing1Motion=Free
  swing2LimitDegrees=45
  swing2Motion=Free
  twistLimitDegrees=45
  twistMotion=Free
 }
]
